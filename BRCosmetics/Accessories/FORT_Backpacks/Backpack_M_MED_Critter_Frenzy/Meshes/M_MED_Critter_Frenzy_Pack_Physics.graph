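# Physics asset for the Critter Frenzy back bling mesh in Battle Royale cosmetics, giving its dyn_Brain and dyn_Water bones physical bodies on the pack.
# Each of the three bones, pack, dyn_Brain and dyn_Water, has a single capsule shape.
# Angular-limited constraints tie both dyn bones directly to the pack, with tight limits on the water and much looser swing and twist on the brain.
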PhysicsAsset /BRCosmetics/Accessories/FORT_Backpacks/Backpack_M_MED_Critter_Frenzy/Meshes/M_MED_Critter_Frenzy_Pack_Physics
bodies {
 dyn_Brain [{
   shapeType=Capsule
   length=2.2041
   radius=7.1902
   shapeName=dyn_Brain_capsule
   center=<0.3147 -0.0507 -0.0332>
   rotation=<rot -0.9994 -90 -37.9917>
  }]
 dyn_Water [{
   shapeType=Capsule
   length=9.2611
   radius=4.4768
   shapeName=dyn_Water_capsule
   center=<-0.6233 3.2569 0.5642>
   rotation=<rot 79.7813 91.7643 -178.2542>
  }]
 pack [{shapeType=Capsule, length=13.1665, radius=9.3969, center=<1.0626 -7.9968 -3.7945>, rotation=<rot -12.2456 -90 24.5169>}]
}
constraints [
 {bone1Name=dyn_Water, bone2Name=pack, swing1LimitDegrees=5, swing2LimitDegrees=5, twistLimitDegrees=5}
 {bone1Name=dyn_Brain, bone2Name=pack, swing1LimitDegrees=30, swing2LimitDegrees=30, twistLimitDegrees=30}
]
