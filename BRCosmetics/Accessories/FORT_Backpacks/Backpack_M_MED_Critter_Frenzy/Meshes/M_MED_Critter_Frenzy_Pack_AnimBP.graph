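# Animation Blueprint for the Critter Frenzy backpack cosmetic, adding secondary motion to its water and brain bones.
# A RigidBody node simulates from the reference pose, then chained spring-driven AnimDynamics nodes move the water and brain bones.
# A weighted constraint links dyn_Brain to the simulated chain, then additive bone edits bob and pulse it and apply a fixed offset to the pack.
# The update event generates that bob and pulse procedurally from time-based sine and cosine waves.
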AnimBlueprint /BRCosmetics/Accessories/FORT_Backpacks/Backpack_M_MED_Critter_Frenzy/Meshes/M_MED_Critter_Frenzy_Pack_AnimBP
target_skeleton=&/M_MED_Critter_Frenzy_Pack_Skeleton
graphs [
 AnimGraph {nodes [
   n1 = RigidBody(ComponentPose=Local To Component(LocalPose=Local Space Ref Pose())) {
    bClampLinearTranslationLimitToRefPose=true
    bOverrideWorldGravity=true
    overridePhysicsAsset=&/M_MED_Critter_Frenzy_Pack_Physics
    simulationSpace=BaseBoneSpace
    alphaBoolBlend {blendOption=Linear}
    baseBoneRef {boneName=root}
    componentAppliedLinearAccClamp=<3000 3000 3000>
    componentLinearAccScale=<0.334 0.334 0.334>
    componentLinearVelScale=<0.334 0.334 0.334>
    overrideWorldGravity {z=-980}
   }
   n2 = Anim Dynamics - Bone: dyn_Water_baseA(ComponentPose=@n1, Alpha=1, ExternalForce="0,0,0") {
    angularDampingOverride=0.999
    angularSpringConstant=800
    bAngularSpring=true
    bGravityOverrideInSimSpace=true
    bOverrideAngularDamping=true
    alphaBoolBlend {blendOption=Linear}
    boundBone {boneName=dyn_Water_baseA}
    componentAppliedLinearAccClamp=<1000 1000 1000>
    componentLinearAccScale=<1 1 1>
    componentLinearVelScale=<1 1 1>
    physicsBodyDefinitions [{
      sphereCollisionRadius=0
      boundBone {boneName=dyn_Water_baseA}
      boxExtents=<3 10 15>
      constraintSetup {angularTargetAxis=AxisY, twistAxis=AxisY, angularLimitsMax=<0 8 0>, angularLimitsMin=<0 -8 -10>, angularTarget=<0 0 0>}
      localJointOffset=<-5 -0.1 -0.01>
     }]
   }
   n3 = Anim Dynamics - Bone: dyn_Water_baseB(ComponentPose=@n2, Alpha=1, ExternalForce="0,0,0") {
    angularDampingOverride=0.999
    angularSpringConstant=800
    bAngularSpring=true
    bGravityOverrideInSimSpace=true
    bOverrideAngularDamping=true
    alphaBoolBlend {blendOption=Linear}
    boundBone {boneName=dyn_Water_baseB}
    componentAppliedLinearAccClamp=<1000 1000 1000>
    componentLinearAccScale=<1 1 1>
    componentLinearVelScale=<1 1 1>
    physicsBodyDefinitions [{
      sphereCollisionRadius=0
      boundBone {boneName=dyn_Water_baseB}
      boxExtents=<3 10 15>
      constraintSetup {angularTargetAxis=AxisY, twistAxis=AxisY, angularLimitsMax=<0 12 0>, angularLimitsMin=<0 -12 -10>, angularTarget=<0 0 0>}
      localJointOffset=<-5 -0.1 -0.01>
     }]
   }
   n4 = Anim Dynamics - Bone: dyn_Water(ComponentPose=@n3, Alpha=1, ExternalForce="0,0,0") {
    angularDampingOverride=0.999
    angularSpringConstant=1000
    bAngularSpring=true
    bGravityOverrideInSimSpace=true
    bOverrideAngularDamping=true
    alphaBoolBlend {blendOption=Linear}
    boundBone {boneName=dyn_Water}
    componentAppliedLinearAccClamp=<1000 1000 1000>
    componentLinearAccScale=<1 1 1>
    componentLinearVelScale=<1 1 1>
    physicsBodyDefinitions [{
      sphereCollisionRadius=0
      boundBone {boneName=dyn_Water}
      boxExtents=<3 10 15>
      constraintSetup {angularTargetAxis=AxisY, twistAxis=AxisY, angularLimitsMax=<2 12 -1>, angularLimitsMin=<-2 -12 -18>, angularTarget=<0 0 0>}
      localJointOffset=<-5 -0.05 -0.01>
     }]
   }
   n5 = Anim Dynamics - Bone: dyn_Brain_rototain_base(ComponentPose=@n4, Alpha=1, ExternalForce="0,0,0") {
    angularDampingOverride=0.999
    angularSpringConstant=800
    bAngularSpring=true
    bGravityOverrideInSimSpace=true
    bOverrideAngularDamping=true
    gravityScale=0.5
    alphaBoolBlend {blendOption=Linear}
    boundBone {boneName=dyn_Brain_rototain_base}
    componentAppliedLinearAccClamp=<10000 10000 10000>
    componentLinearAccScale=<0.1 0.1 0.1>
    componentLinearVelScale=<0.1 0.1 0.1>
    physicsBodyDefinitions [{
      sphereCollisionRadius=0
      boundBone {boneName=dyn_Brain_rototain_base}
      boxExtents=<3 10 5>
      constraintSetup {angularTargetAxis=AxisY, twistAxis=AxisY, angularLimitsMax=<0 2 7>, angularLimitsMin=<0 -2 -3>, angularTarget=<0 0 0>}
      localJointOffset=<-2 -0.05 0>
     }]
   }
   n6 = Anim Dynamics - Bone: dyn_Brain_rototain(ComponentPose=@n5, Alpha=1, ExternalForce="0,0,0") {
    angularDampingOverride=0.999
    angularSpringConstant=800
    bAngularSpring=true
    bGravityOverrideInSimSpace=true
    bOverrideAngularDamping=true
    gravityScale=0.5
    alphaBoolBlend {blendOption=Linear}
    boundBone {boneName=dyn_Brain_rototain}
    componentAppliedLinearAccClamp=<10000 10000 10000>
    componentLinearAccScale=<0.1 0.1 0.1>
    componentLinearVelScale=<0.1 0.1 0.1>
    physicsBodyDefinitions [{
      sphereCollisionRadius=0
      boundBone {boneName=dyn_Brain_rototain}
      boxExtents=<3 10 5>
      constraintSetup {angularTargetAxis=AxisY, twistAxis=AxisY, angularLimitsMax=<0 2 9>, angularLimitsMin=<0 -2 -6>, angularTarget=<0 0 0>}
      localJointOffset=<-2 -0.05 0>
     }]
   }
   n7 = Anim Dynamics - Bone: dyn_Brain_rototain_01(ComponentPose=@n6, Alpha=1, ExternalForce="0,0,0") {
    angularDampingOverride=0.999
    angularSpringConstant=800
    bAngularSpring=true
    bGravityOverrideInSimSpace=true
    bOverrideAngularDamping=true
    gravityScale=0.5
    alphaBoolBlend {blendOption=Linear}
    boundBone {boneName=dyn_Brain_rototain_01}
    componentAppliedLinearAccClamp=<2000 2000 2000>
    componentLinearAccScale=<0.1 0.1 0.1>
    componentLinearVelScale=<0.1 0.1 0.1>
    physicsBodyDefinitions [{
      sphereCollisionRadius=0
      boundBone {boneName=dyn_Brain_rototain_01}
      boxExtents=<3 10 15>
      constraintSetup {angularTargetAxis=AxisY, twistAxis=AxisY, angularLimitsMax=<10 5 10>, angularLimitsMin=<-10 -5 -10>, angularTarget=<0 0 0>}
      localJointOffset=<-5 -0.05 0>
     }]
   }
   n8 = "Constraint - dyn_Brain (PP)"!(ComponentPose=@n7, Alpha=1, ConstraintWeights_0=0.8, ConstraintWeights_1:float=0.2) -> (Pose) {
    alphaBoolBlend {blendOption=Linear}
    boneToModify {boneName=dyn_Brain}
    constraintSetup [{perAxis {bY=false, bZ=false}, targetBone {boneName=dyn_Brain_rototain_01}}, {targetBone {boneName=dyn_Brain}}]
    constraintWeights [1,1]
   }
   n9 = "Transform (Modify) Bone - Bone: dyn_Brain"(
    ComponentPose=@n8
    Alpha=1
    Translation=Get Brain_A() -> (Brain_A:Vector)
    Rotation="0,0,0"
    Scale=Get Brain_A_scale() -> (Brain_A_scale:Vector)
   ) {
    rotationMode=BMM_Additive
    rotationSpace=BCS_BoneSpace
    scaleMode=BMM_Additive
    scaleSpace=BCS_BoneSpace
    translationMode=BMM_Additive
    translationSpace=BCS_BoneSpace
    alphaBoolBlend {blendOption=Linear}
    boneToModify {boneName=dyn_Brain}
   }
   n10 = "Transform (Modify) Bone - Bone: pack"(ComponentPose=@n9, Translation="0,0.2,0", Rotation="0,0,11") {
    rotationMode=BMM_Additive
    rotationSpace=BCS_BoneSpace
    scaleMode=BMM_Additive
    scaleSpace=BCS_BoneSpace
    translationMode=BMM_Additive
    translationSpace=BCS_BoneSpace
    alphaBoolBlend {blendOption=Linear}
    boneToModify {boneName=pack}
    rotation {roll=11}
    translation {y=0.2}
   }
   Output Pose(Result=Component To Local(ComponentPose=@n10)) {layerGroup=None}
  ]}
 EventGraph {nodes [
   Event BlueprintUpdateAnimation()
   n1 = float * float(A=1, B=1)
   n2 = GetRealTimeSeconds()
   n3 = float * float(A="Sin (Radians)"(A=float * float(A=@n1, B=@n2)), B=0.07)
   n4 = "Sin (Radians)"(A=float * float(A=float * float(A=1, B=2), B=@n2))
   n5 = float * float(A=1, B=1.1)
   n6 = float * float(A="Sin (Radians)"(A=float * float(A=@n5, B=@n2)), B=0.45)
   > Set Brain_A(Brain_A:Vector=MakeVector(
     X=float + float(A=@n6, B=1.3) {is=title K2Node_CommutativeAssociativeBinaryOperator}
     Y=float + float(A=@n3, B=-0.1)
     Z=float * float(A=@n4, B=0.12)
    )) -> (Output_Get:Vector)
   n7 = float * float(A=1, B=0.3)
   n8 = GetGameTimeInSeconds()
   n9 = float * float(A="Cos (Radians)"!(A:double=float * float(A=@n7, B=@n8)) -> (ReturnValue:double), B=0.02)
   n10 = float * float(A=1, B=0.6)
   n11 = float * float(A="Cos (Radians)"(A=float * float(A=@n10, B=@n8)), B=0.02)
   n12 = float * float(A=1, B=0.85)
   n13 = float * float(A="Cos (Radians)"(A=float * float(A=@n12, B=@n8)), B=0.02)
   > Set Brain_A_scale(Brain_A_scale:Vector=MakeVector(X=float + float(A=@n11, B=0.95), Y=float + float(A=@n13, B=0.95), Z=float + float(A=@n9, B=0.95))) -> (Output_Get:Vector)
   TryGetPawnOwner()
  ]}
]
variables [
 Brain_A {type=Vector}
 Brain_A_scale {type=Vector}
]
